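# Unlit material for sticker effects on the DollEnthusiast pickaxe, built for Niagara mesh particles.
# A particle-driven index offsets UVs into the sticker texture, whose green channel blends particle color toward an edge color, with optional vertex-color tint.
# A panning plasma texture in lat-long reflection coordinates fakes reflections.
# A second dynamic parameter rotates, pans and scales a gradient sheen, summed with both into an emissive adjustment function.
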Material /BRCosmetics/Weapons/FORT_Melee/Pickaxe_DollEnthusiast/FX/M_FX_Pickaxe_DollEnthusiast_Stickers
graph {
 properties {assetImportData="<self>:InterchangeAssetImportData_0", bUsedWithNiagaraMeshParticles=true, shadingModel=MSM_Unlit, usedShadingModels=Unlit}
 nodes [
  n1 = Constant()
  n2 = DynamicParameter() {defaultValue=<rgba 0 1 1 1>, paramNames [Index, ReflectionStrength, Param3, Param4]}
  n3 = TextureSample(UVs=Add(
    A=TextureCoordinate() {vTiling=0.5}
    B=MaterialFunctionCall(R=@n1, G=@n2.Index) {materialFunction=&AppendMany}.RG
   )) {texture=&/T_FX_DollEnthiast_Stickers}
  n4 = ScalarParameter() {defaultValue=1, parameterName=ColorEdge}
  n5 = Multiply(
   A=LinearInterpolate(B=VertexColor(), Alpha=ScalarParameter() {defaultValue=1, parameterName=MultiplyByVertColor}) {constA=1}
   B=LinearInterpolate(
    A=ParticleColor().RGB
    B=VectorParameter() {parameterName=EdgeColor, defaultValue=<rgba 0 0 0 1>}.RGB
    Alpha=LinearInterpolate(B=@n3.G, Alpha=@n4)
   )
  )
  n6 = ReflectionVectorWS(CustomWorldNormal=CameraVectorWS())
  n7 = Panner(Coordinate=Multiply(A=MaterialFunctionCall!(CameraVector=@n6) {materialFunction=/Engine/Functions/Engine_MaterialFunctions01/Texturing/CameraVectorToLatLongUV}.Uvs) {constB=3}) {speedY=1}
  n8 = Multiply(
   A=Multiply(A=VectorParameter() {parameterName=ReflectionColor, defaultValue=<rgba 0 0 0 1>}.RGB, B=@n2.ReflectionStrength)
   B=Power(Base=TextureSample(UVs=@n7) {texture=/Game/Effects/Fort_Effects/Textures/T_FX_SpringerPlasma}.G) {constExponent=3}
  )
  MaterialFunctionCall(Emissive=Add(A=NamedRerouteUsage(), B=Add(A=@n5, B=@n8))) {materialFunction=&MF_FX_EmissiveAdjustments}
  n9 = DynamicParameter() {parameterIndex=1, defaultValue=<rgba 0 1 1 1>, paramNames [Index/SheenRotation, ReflectionStrength/SheenPan, Param3/SheenStrength, Param4]}
  n10 = Rotator(Coordinate=TextureCoordinate(), Time=@n9.Index/SheenRotation) {speed=1}
  n11 = TextureSample(UVs=Multiply(A=Panner(Coordinate=@n10, Time=@n9.ReflectionStrength/SheenPan) {speedY=1})) {
   samplerSource=SSM_Clamp_WorldGroupSettings
   texture=/BRCosmetics/Weapons/FORT_Melee/Pickaxe_GoldenGuard/FX/T_Gradient_SharpBladeSheen_GoldenGuard
  }
  n12 = VectorParameter() {parameterName=SheenColor, defaultValue=<rgba 0 0 0 1>}
  NamedRerouteDeclaration(None=Multiply(A=Multiply(A=@n11.RGB, B=@n12.RGB), B=@n9.Param3/SheenStrength)) {name=Sheen, nodeColor=<rgba 1 0.8203 0 1>}
 ]
}
