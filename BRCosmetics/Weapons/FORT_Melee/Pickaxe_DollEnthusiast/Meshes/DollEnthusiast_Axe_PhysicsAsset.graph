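# Physics asset for the DollEnthusiast pickaxe mesh, a melee weapon cosmetic with an accordion section.
# Three capsule bodies cover the accordion's middle bone and its left and right "dyn_" end bones.
# Two constraints join each end to the middle with twist and both swings locked, so the angular limit values do not give any rotational freedom.
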PhysicsAsset /BRCosmetics/Weapons/FORT_Melee/Pickaxe_DollEnthusiast/Meshes/DollEnthusiast_Axe_PhysicsAsset
bodies {
 accordion_middle [{shapeType=Capsule, length=7.5, radius=14, shapeName=accordion_middle_capsule, center=<0 -0.0002 13.4741>, rotation=<rot 0 0 -90.0009>}]
 dyn_accordion_end_l [{shapeType=Capsule, length=2, radius=14, shapeName=dyn_accordion_end_l_capsule, center=<12 0 0>, rotation=<rot 90 180 180>}]
 dyn_accordion_end_r [{shapeType=Capsule, length=2, radius=14, shapeName=dyn_accordion_end_r_capsule, center=<12 0 0>, rotation=<rot -90 0 0>}]
}
constraints [
 {
  bone1Name=dyn_accordion_end_r
  bone2Name=accordion_middle
  swing1LimitDegrees=45
  swing1Motion=Locked
  swing2LimitDegrees=45
  swing2Motion=Locked
  twistLimitDegrees=45
  twistMotion=Locked
 }
 {
  bone1Name=dyn_accordion_end_l
  bone2Name=accordion_middle
  swing1LimitDegrees=45
  swing1Motion=Locked
  swing2LimitDegrees=45
  swing2Motion=Locked
  twistLimitDegrees=45
  twistMotion=Locked
 }
]
